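; Default knockback gameplay ability for the fourth hit of ChapWorm melee, built on a parent knockback ability.
; It overrides the initial knock magnitude to pass the launcher's forward vector to the parent, or the given direction without a launcher.
; Activation requires the weapon owner to lack the ChapWorm.Overpower tag and the knockback hotfix to be enabled, despite an inline comment stating the opposite tag condition.
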(event Ability|EventActivateAbility
  (|CallParentFunction) ; This node is disabled and will not be called.\nDrag off pins to build functionality.
)

(event Ability|EventOnEndAbility
  (params (bWasCancelled :type Boolean)))

(fn GetInitialKnockMag
  (params (Direction :type Vector :default "0, 0, 0") (OverrideXYMag :type Boolean) (XYMAgOverride :type Float%20%28double-precision%29))
  ;[0.05881 0.1823 0.1787] GetInitialKnockbackMagnitude: Knock the player back in the direction that the launcher is facing.
  (Utilities|FlowControl|Sequence
    (:then_0
      (bind _launcher (Variables|KnockbackParent|GetLauncher)
        (:then
          (return (|CallParentFunction (Transformation|GetActorForwardVector _launcher))))
        (:else)))
    (:then_1
      (return (|CallParentFunction Direction)))))

(fn CanActivateAbility
  (params (ActorInfo :type Gameplay%20Ability%20Actor%20Info%20Structure) (Handle :type Gameplay%20Ability%20Spec%20Handle%20Structure))
  ;[0.01653 0.09686 0.151] CanActivateAbility: Ability can only be activated if the attacking player is overpowered.
  (Utilities|FlowControl|Sequence
    (:then_0
      (bind _asfort_weapon (Utilities|Casting|CastToFortWeapon (Ability|GetSourceObjectNonInstanced Handle ActorInfo)))
      (bind _asfort_player_pawn (Utilities|Casting|CastToFortPlayerPawn (Actor|GetOwner _asfort_weapon)))
      (return 0 (and (not (GameplayTags|HasMatchingGameplayTag _asfort_player_pawn "(TagName=\"ChapWorm.Overpower\")")) (ScalableFloat|GetValueasBool (Variables|Hotfix|GetHotfix_KnockbackEnabled)))))
    (:then_1
      (bind (_relevanttags _returnvalue) (|CallParentFunction ActorInfo Handle))
      (return _relevanttags _returnvalue))))
