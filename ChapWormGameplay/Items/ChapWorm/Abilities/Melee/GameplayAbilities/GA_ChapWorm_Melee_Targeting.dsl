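; Melee targeting ability for ChapWorm that recovers a held primary attack input that may have been ignored during a pounce.
; On activation it waits for the ChapWorm.Ability.Melee.Blocked tag to be removed, with a hotfixable delay before simulating input.
; On non-authority, a locally controlled pawn holding primary outside a vehicle gets a simulated Primary press through its Mantis component.
; Ending the ability cancels the tag-wait task.
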(event Ability|EventActivateAbility
  (bind _hotfix_simulateclientmeleeinputdelay (Variables|Hotfix|GetHotfix_SimulateClientMeleeInputDelay))
  (bind _returnvalue (ScalableFloat|GetValueatLevel _hotfix_simulateclientmeleeinputdelay))
  ;[0.03555 0.15 0.05254 0.5] ActivateAbility: When the player starts targeting melee again (say after a pounce attack), if the player has the primary input held, we should simulate an input for begin the melee attack cause the primary input may have been ignored while performing the pounce attack.
  ; Say the player holds the melee attack while pouncing but the melee attack never occurs after the pounce cause the initial input was ignored.
  (|CallParentFunction)
  (bind _asynctaskproxy (Ability|Tasks|WaitGameplayTagRemove "(TagName=\"ChapWorm.Ability.Melee.Blocked\")")
    (:then
      (Variables|Transient|SetAsyncTask_GameplayTagRemoved _asynctaskproxy))
    (:Removed
      (if (>= _returnvalue 0)
        (Utilities|FlowControl|Delay _returnvalue)
        (else
          _))
      (Default|SimulateClientMeleeInput))) ; This tag is removed when the player is ready to attack again after the pounce.
)

(event Ability|EventOnEndAbility
  (params (bWasCancelled :type Boolean))
  ;[0.09896 0.05222 0.01087 0.5] OnEndAbility: Clean up.
  (Utilities|FlowControl|Sequence
    (:then_0
      (bind _asynctask_gameplaytagremoved (Variables|Transient|GetAsyncTask_GameplayTagRemoved)
        (:then
          (GameplayTasks|EndTask _asynctask_gameplaytagremoved))
        (:else)))
    (:then_1
      (|CallParentFunction bWasCancelled))))

(fn SimulateClientMeleeInput
  ;[0.1531 0.1804 1] If the player is holding down the primary input when targeting starts (and NOT in a vehicle), perform a melee attack.
  (Ability|HasAuthority
    (:False
      (bind _outfortplayerpawn (CallFunction|GetAvatarasValidFortPlayerPawn)
        (:ValidFortPlayerPawn
          (if (Math|Boolean|ANDBoolean (Variables|/Script/FortniteGame.FortPawn|GetbPrimaryInputHeld _outfortplayerpawn) (Pawn|IsLocallyControlled _outfortplayerpawn) (not (Vehicle|IsinVehicle _outfortplayerpawn)))
            (Utilities|IsValid (Actor|GetComponentbyClass _outfortplayerpawn "/Script/MantisRuntime.FortMantisPawnComponent")
              (:Is%20Valid
                (Mantis|SimulateInputEvent (Actor|GetComponentbyClass _outfortplayerpawn "/Script/MantisRuntime.FortMantisPawnComponent") "Primary" "Pressed" true false)))))))))
